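; Season 20 narrative quest ability that sends the player from a starting location to a destination POI.
; On activation, it checks the quest objective and starting location tags, removes the current location from the destinations, and grants a transient urgent quest.
; It then starts a WaitGameplayTagAdd task per destination tag, with debug prints.
; Gaining a destination tag sends a quest stat event and ends the ability.
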(event Ability|EventActivateAbilityFromEvent
  (params (EventData :type Gameplay%20Event%20Data%20Structure%20%28by%20ref%29))
  ;[0 0 0] Activate ability - check if player is in correct starting location and on the right quest objective
  (|CallParentFunction EventData)
  (if (Quest|IsObjectivewithNameinProgress (Variables|GAVars|Data|GetOwningPlayerQuestManager) (Variables|Quest|GetQuestItemDefinition) "quest_s20_narrative_w11_q04_obj1" :!class /Script/FortniteGame.FortQuestManager)
    (if (GameplayTags|HasAnyMatchingGameplayTags (Variables|GAVars|Data|GetOwningPlayerPawn) (Variables|Quest|GetStartingLocationTags))
      (bind _array_element (Utilities|Array|ForEachLoopwithBreak (GameplayTags|BreakGameplayTagContainer (Variables|Quest|GetStartingLocationTags)))
        (:LoopBody
          (if (GameplayTags|HasMatchingGameplayTag (Variables|GAVars|Data|GetOwningPlayerPawn) _array_element)
            (GameplayTags|RemoveGameplayTag (Variables|Quest|GetDestinationTags) _array_element :!class /Script/GameplayTags.BlueprintGameplayTagLibrary)))
        (:Completed
          _))
      (CallFunction|GrantUrgentQuesttoOwningPlayer)
      (CallFunction|HandleEnteringDestination)
      (else
        _))
    (else
      _))
  (Ability|EndAbility))

(event Custom|HandleEnteringDestination
  ;[0 0 0] Listen for destination tag to be added
  (for _array_element (GameplayTags|BreakGameplayTagContainer (Variables|Quest|GetDestinationTags))
    (Ability|Tasks|WaitGameplayTagAdd _array_element 0 true
      (:then
        (Development|PrintString (Utilities|String|Append "Waiting on " (Utilities|String|ToString(Name) (GameplayTags|GetTagName _array_element))) true true "(R=1.000000,G=0.000000,B=0.189399,A=1.000000)" 3.0))
      (:Added
        (Quest|SendComplexCustomStatEvent (Variables|GAVars|Data|GetOwningPlayerQuestManager) 0 (Variables|Quest|GetQuestSourceTags) "(GameplayTags=)" 1)
        (Ability|EndAbility)))))

(fn GetPassengersinOwningPlayerVehicle
  ;[0 0 0] Helper function for getting all the passengers currently in the owning player's vehicle.
  (return (Vehicle|GetAllPassengers (Vehicle|GetVehicleInterface (Variables|GAVars|Data|GetOwningPlayerPawn)))))

(fn GrantUrgentQuestToOwningPlayer
  ;[0 0 0] Helper function for granting the specified urgent quest to the ability owner
  (bind _asfort_player_controller_athena (Utilities|Casting|CastToFortPlayerControllerAthena (Fortnite|GetFortPlayerControllerfromActor (Variables|GAVars|Data|GetOwningPlayerPawn)))
    (:then
      (Class|FortControllerComponentTransientQuests|GrantTransientQuest (Class|FortPlayerControllerAthena|GetTransientQuestsComponent _asfort_player_controller_athena) (Variables|Quest|GetUrgentQuestToGrant) 0)
      (return))
    (:CastFailed
      (return))))
