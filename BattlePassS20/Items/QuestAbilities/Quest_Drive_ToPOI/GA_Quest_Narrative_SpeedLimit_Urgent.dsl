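; Passive quest ability that fails an urgent quest when the player drives over a speed limit or leaves the vehicle.
; On the server, it verifies valet vehicle tags, then starts a repeating timer that polls forward speed.
; Nearing the limit applies gameplay effects that cycle through three voice barks, throttled by a cooldown.
; Overspeed shows a HUD failure message, clears the timer and fails the quest; vehicle exit fails it directly.
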(event Ability|EventFailedToActivatePassiveAbility
  (params (ActorInfo :type Gameplay%20Ability%20Actor%20Info%20Structure))
  (|CallParentFunction))

(event Ability|EventActivateAbility
  ;[0 0 0] Activate Ability
  (|CallParentFunction)
  (if (Networking|IsServer)
    (Ability|Tasks|WaitGameplayTagRemove "(TagName=\"Gameplay.Action.Player.AthenaInVehicle\")" 0 true
      (:then
        ;;[0 0 0] Check if player has correct vehicle tag(s)
        (if (GameplayTags|TagContainer|CompareTagAssetInterfacetoOtherTags (Variables|Default|GetValetVehicle) (Vehicle|GetVehicleAbilitySystemComponent (Vehicle|GetVehicleInterface (CallFunction|GetAvatarasFortPlayerPawn))))
          (bind _returnvalue (Utilities|Time|SetTimerbyEvent (EventDispatchers|CreateEvent :!event Custom|CheckSpeedLimit) (Variables|Default|GetTimeInterval) true))
          (Variables|Default|SetTimerHandle _returnvalue)
          (else
            (Development|PrintString "Vehicle did not have tags" true true "(R=1.000000,G=0.000000,B=0.278642,A=1.000000)" 4.0)
            (CallFunction|Cleanup))))
      (:Removed
        ;;[0 0 0] Opt-out the urgent quest when the player exits the vehicle
        (CallFunction|FailQuest)
        (Development|PrintString "Player left vehicle- failing urgent quest" true true "(R=1.000000,G=0.000000,B=0.278642,A=1.000000)" 4.0)))))

(event Custom|CheckSpeedLimit
  ;[0 0 0] If speed surpasses maximum limit, display an urgent quest message and fail the quest
  (Variables|Default|SetTimeElapsed (+ (Variables|Default|GetTimeElapsed) (Variables|Default|GetTimeInterval)))
  (if (and (> (Variables|Default|GetTimeElapsed) 2e+01) (Variables|Default|GetbBarkPlayed))
    (Variables|Default|SetbBarkPlayed false)
    (else
      _))
  (bind _asfort_athena_vehicle (Utilities|Casting|CastToFortAthenaVehicle (Vehicle|GetVehicleInterface (CallFunction|GetAvatarasFortPlayerPawn))))
  (if (> (Vehicle|GetForwardSpeedKmh _asfort_athena_vehicle) (Variables|Default|GetMaxSpeedLimit))
    (CallFunction|ShowMessageFailure)
    (CallFunction|Cleanup)
    (CallFunction|FailQuest)
    (elif (and (not (Variables|Default|GetbBarkPlayed)) (> (Vehicle|GetForwardSpeedKmh _asfort_athena_vehicle) (Variables|Default|GetBarkTriggerSpeed)))
      (GameplayEffects|ApplyGameplayEffectToSelf :self (Ability|GetFortAbilitySystemComponent (CallFunction|GetAvatarasFortPlayerPawn)) :GameplayEffectClass (Utilities|Array|Get(acopy) (Variables|Default|GetVO_Bark_Trigger) (Variables|Default|GetBarkNumber)) :Level 0.0)
      (Variables|Default|SetBarkNumber (+ (Variables|Default|GetBarkNumber) 1))
      (if (> (Variables|Default|GetBarkNumber) 2)
        (Variables|Default|SetBarkNumber 0)
        (else
          _))
      (Variables|Default|SetbBarkPlayed true)
      (Variables|Default|SetTimeElapsed 0.0))))

(event Custom|ShowMessage_Failure
  (bind (_eventtag _visualstyletag _totaleventtime) (Utilities|Struct|BreakUrgentQuestData (Class|FortUrgentQuestItemDefinition|GetUrgentQuestData (Variables|Default|GetUrgentQuestDefinition))))
  ;[0 0 0] Show a HUD message to the owning client if they exceed max speed and fail the quest
  (Class|AthenaHUDContext|ShowUrgentQuestMessage (BlueprintContext|GetAthenaHUDContext) (Utilities|Struct|MakeUrgentQuestData _eventtag "(GameplayTags=)" _totaleventtime true false false false false true "" (Variables|Default|GetTextQuestFailed) (Variables|Default|GetTextSpeedExceeded))))

(event Custom|Cleanup
  (bind _timerhandle (Variables|Default|GetTimerHandle))
  ;[0 0 0] Cleanup - Clear and invalidate Timer
  (if (Utilities|Time|IsValidTimerHandle _timerhandle)
    (Utilities|Time|ClearandInvalidateTimerbyHandle _timerhandle)))
